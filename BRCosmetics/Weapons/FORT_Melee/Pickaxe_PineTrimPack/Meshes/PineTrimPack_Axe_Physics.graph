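# Physics asset for the PineTrimPack two-handed pickaxe, simulating a dangling cord on the axe.
# Three capsules cover the handle bone, anchoring a Base_Cord body that links to Dyn_Cord.
# Dyn_Cord forks into two three-bone capsule chains, CordUP and CordDwon, each ending in a larger tip body.
# Constraints between neighboring bones limit swing and set twist limits to zero degrees.
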PhysicsAsset /BRCosmetics/Weapons/FORT_Melee/Pickaxe_PineTrimPack/Meshes/PineTrimPack_Axe_Physics
bodies {
 Base_Cord [{shapeType=Capsule, length=0.505, radius=0.505, shapeName=Base_Cord_capsule, rotation=<rot 90 0 0>}]
 Dyn_Cord [{shapeType=Capsule, length=1.1462, radius=1, shapeName=Dyn_Cord_capsule, center=<1 0 0>, rotation=<rot 90 0 0>}]
 Dyn_CordUP_01 [{shapeType=Capsule, length=3.8, radius=0.5, shapeName=Dyn_CordUP_01_capsule, center=<1 0 0>, rotation=<rot 90 0 0>}]
 Dyn_CordUP_02 [{shapeType=Capsule, length=3.5218, radius=0.5, shapeName=Dyn_CordUP_02_capsule, center=<2 0.2 0>, rotation=<rot 90 0 0>}]
 Dyn_CordUP_03 [{shapeType=Capsule, length=1.7777, radius=2.5223, shapeName=Dyn_CordUP_03_capsule, center=<6 0 0>, rotation=<rot 90 0 0>}]
 Melee_TwoHanded_Handle [
  {shapeType=Capsule, length=0.0601, radius=25.3486, center=<0.5836 0 72.4173>}
  {shapeType=Capsule, length=73.2487, radius=8.4421, shapeName=Melee_TwoHanded_Handle_capsule, center=<1.8005 -7.2259 18.0425>}
  {shapeType=Capsule, length=0, radius=7.017, shapeName=Melee_TwoHanded_Handle_capsule, center=<0 -1.7804 37.5269>}
 ]
 dyn_CordDwon_01 [{shapeType=Capsule, length=2.5, radius=0.5434, shapeName=dyn_CordDwon_01_capsule, center=<1 0 0>, rotation=<rot 90 0 0>}]
 dyn_CordDwon_02 [{shapeType=Capsule, length=3.4739, radius=0.5, shapeName=dyn_CordDwon_02_capsule, center=<1 0 0>, rotation=<rot 90 0 0>}]
 dyn_CordDwon_03 [{shapeType=Capsule, length=8.5231, radius=0.7, shapeName=dyn_CordDwon_03_capsule, center=<4.1475 0 0>, rotation=<rot 90 0 0>}]
}
constraints [
 {bone1Name=dyn_CordDwon_03, bone2Name=dyn_CordDwon_02, swing1LimitDegrees=40, swing2LimitDegrees=60, twistLimitDegrees=0}
 {bone1Name=Dyn_CordUP_02, bone2Name=Dyn_CordUP_01, swing1LimitDegrees=40, swing2LimitDegrees=60, twistLimitDegrees=0}
 {bone1Name=Dyn_CordUP_03, bone2Name=Dyn_CordUP_02, swing1LimitDegrees=40, swing2LimitDegrees=60, twistLimitDegrees=0}
 {bone1Name=Dyn_CordUP_01, bone2Name=Dyn_Cord, swing1LimitDegrees=50, swing2LimitDegrees=65, twistLimitDegrees=0}
 {bone1Name=dyn_CordDwon_02, bone2Name=dyn_CordDwon_01, swing1LimitDegrees=45, swing2LimitDegrees=60, twistLimitDegrees=0}
 {bone1Name=Dyn_Cord, bone2Name=Base_Cord, swing1LimitDegrees=70, swing2LimitDegrees=90, twistLimitDegrees=0}
 {bone1Name=dyn_CordDwon_01, bone2Name=Dyn_Cord, swing1LimitDegrees=50, swing2LimitDegrees=65, twistLimitDegrees=0}
 {bone1Name=Base_Cord, bone2Name=Melee_TwoHanded_Handle, swing1LimitDegrees=50, swing2LimitDegrees=70, twistLimitDegrees=0}
]
