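# Animation Blueprint for the PineTrimPack pickaxe that adds secondary motion to its hanging cord bones.
# It uses RigidBody simulation, switching to an AnimDynamics chain with lower-cord offsets in the front end or when rigid bodies are disabled.
# Base_Cord rotation offsets and Trail Controllers follow, selected by IsTrailOnPawn (on pawn, both simulations disabled) or its negation IsNotOnPawn, so one trail set always runs.
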AnimBlueprint /BRCosmetics/Weapons/FORT_Melee/Pickaxe_PineTrimPack/Meshes/PineTrimPack_Axe_AnimBP
target_skeleton=&/PineTrimPack_Axe_Skeleton
graphs [
 AnimGraph {nodes [
   n1 = Anim Dynamics - Bone: Dyn_Cord(ComponentPose=Local To Component(LocalPose=Use cached pose 'CRCache_InputPose'()), Alpha=0.9, ExternalForce="0,0,0") {
    angularDampingOverride=1
    angularSpringConstant=50
    bAngularSpring=true
    bOverrideAngularDamping=true
    bUseGravityOverride=true
    gravityScale=0.35
    alphaBoolBlend {blendOption=Linear}
    boundBone {boneName=Dyn_Cord}
    componentAppliedLinearAccClamp=<5000 5000 5000>
    componentLinearAccScale=<0.1 0.1 0.1>
    componentLinearVelScale=<0.1 0.1 0.1>
    gravityOverride {z=-980}
    physicsBodyDefinitions [{
      boundBone {boneName=Dyn_Cord}
      boxExtents=<10 30 30>
      constraintSetup {angularLimitsMax=<0 90 45>, angularLimitsMin=<0 -90 -60>, angularTarget=<1 0 0>}
      localJointOffset=<10 0 0>
     }]
    relativeSpaceBone {boneName=Base_Cord}
    simSpaceSettings {maxAngularAcceleration=500, maxAngularVelocity=200, simSpaceAngularAlpha=0.65}
   }
   n2 = Anim Dynamics - Bone: Dyn_CordUP_01(ComponentPose=@n1, Alpha=0.8, ExternalForce="0,0,0") {
    angularDampingOverride=1
    angularSpringConstant=50
    bAngularSpring=true
    bOverrideAngularDamping=true
    bUseGravityOverride=true
    gravityScale=0.35
    alphaBoolBlend {blendOption=Linear}
    boundBone {boneName=Dyn_CordUP_01}
    componentAppliedLinearAccClamp=<5000 5000 5000>
    componentLinearAccScale=<0.1 0.1 0.1>
    componentLinearVelScale=<0.1 0.1 0.1>
    gravityOverride {z=-980}
    physicsBodyDefinitions [{
      boundBone {boneName=Dyn_CordUP_01}
      boxExtents=<10 30 30>
      constraintSetup {angularLimitsMax=<0 90 10>, angularLimitsMin=<0 -90 -10>, angularTarget=<1 0 0>}
      localJointOffset=<10 0 0>
     }]
    relativeSpaceBone {boneName=Base_Cord}
    simSpaceSettings {maxAngularAcceleration=500, maxAngularVelocity=200, simSpaceAngularAlpha=0.65}
   }
   n3 = Anim Dynamics - Bone: Dyn_CordUP_02(ComponentPose=@n2, Alpha=0.8, ExternalForce="0,0,0") {
    angularDampingOverride=1
    angularSpringConstant=50
    bAngularSpring=true
    bOverrideAngularDamping=true
    bUseGravityOverride=true
    gravityScale=0.35
    alphaBoolBlend {blendOption=Linear}
    boundBone {boneName=Dyn_CordUP_02}
    componentAppliedLinearAccClamp=<5000 5000 5000>
    componentLinearAccScale=<0.1 0.1 0.1>
    componentLinearVelScale=<0.1 0.1 0.1>
    gravityOverride {z=-980}
    physicsBodyDefinitions [{
      boundBone {boneName=Dyn_CordUP_02}
      boxExtents=<10 30 30>
      constraintSetup {angularLimitsMax=<0 90 10>, angularLimitsMin=<0 -90 -10>, angularTarget=<1 0 0>}
      localJointOffset=<10 0 0>
     }]
    relativeSpaceBone {boneName=Base_Cord}
    simSpaceSettings {maxAngularAcceleration=500, maxAngularVelocity=200, simSpaceAngularAlpha=0.65}
   }
   n4 = Anim Dynamics - Bone: Dyn_CordUP_03(ComponentPose=@n3, Alpha=0.8, ExternalForce="0,0,0") {
    angularDampingOverride=1
    angularSpringConstant=50
    bAngularSpring=true
    bOverrideAngularDamping=true
    bUseGravityOverride=true
    gravityScale=0.35
    alphaBoolBlend {blendOption=Linear}
    boundBone {boneName=Dyn_CordUP_03}
    componentAppliedLinearAccClamp=<5000 5000 5000>
    componentLinearAccScale=<0.1 0.1 0.1>
    componentLinearVelScale=<0.1 0.1 0.1>
    gravityOverride {z=-980}
    physicsBodyDefinitions [{
      boundBone {boneName=Dyn_CordUP_03}
      boxExtents=<10 30 30>
      constraintSetup {angularLimitsMax=<0 90 10>, angularLimitsMin=<0 -90 -10>, angularTarget=<1 0 0>}
      localJointOffset=<10 0 0>
     }]
    relativeSpaceBone {boneName=Base_Cord}
    simSpaceSettings {maxAngularAcceleration=500, maxAngularVelocity=200, simSpaceAngularAlpha=0.65}
   }
   n5 = Anim Dynamics - Bone: dyn_CordDwon_01(ComponentPose=@n4, Alpha=0.9, ExternalForce="0,0,0") {
    angularDampingOverride=1
    angularSpringConstant=50
    bAngularSpring=true
    bOverrideAngularDamping=true
    bUseGravityOverride=true
    gravityScale=0.35
    alphaBoolBlend {blendOption=Linear}
    boundBone {boneName=dyn_CordDwon_01}
    componentAppliedLinearAccClamp=<5000 5000 5000>
    componentLinearAccScale=<0.1 0.1 0.1>
    componentLinearVelScale=<0.1 0.1 0.1>
    gravityOverride {z=-980}
    physicsBodyDefinitions [{
      boundBone {boneName=dyn_CordDwon_01}
      boxExtents=<10 30 30>
      constraintSetup {angularLimitsMax=<0 70 10>, angularLimitsMin=<0 -70 -10>, angularTarget=<1 0 0>}
      localJointOffset=<10 0 0>
     }]
    relativeSpaceBone {boneName=Base_Cord}
    simSpaceSettings {maxAngularAcceleration=500, maxAngularVelocity=200, simSpaceAngularAlpha=0.65}
   }
   n6 = Anim Dynamics - Bone: dyn_CordDwon_02(ComponentPose=@n5, Alpha=0.8, ExternalForce="0,0,0") {
    angularDampingOverride=1
    angularSpringConstant=50
    bAngularSpring=true
    bOverrideAngularDamping=true
    bUseGravityOverride=true
    gravityScale=0.35
    alphaBoolBlend {blendOption=Linear}
    boundBone {boneName=dyn_CordDwon_02}
    componentAppliedLinearAccClamp=<5000 5000 5000>
    componentLinearAccScale=<0.1 0.1 0.1>
    componentLinearVelScale=<0.1 0.1 0.1>
    gravityOverride {z=-980}
    physicsBodyDefinitions [{
      boundBone {boneName=dyn_CordDwon_02}
      boxExtents=<10 30 30>
      constraintSetup {angularLimitsMax=<0 70 10>, angularLimitsMin=<0 -70 -10>, angularTarget=<1 0 0>}
      localJointOffset=<10 0 0>
     }]
    relativeSpaceBone {boneName=Base_Cord}
    simSpaceSettings {maxAngularAcceleration=500, maxAngularVelocity=200, simSpaceAngularAlpha=0.65}
   }
   n7 = Anim Dynamics - Bone: dyn_CordDwon_03(ComponentPose=@n6, Alpha=0.8, ExternalForce="0,0,0") {
    angularDampingOverride=1
    angularSpringConstant=50
    bAngularSpring=true
    bOverrideAngularDamping=true
    bUseGravityOverride=true
    gravityScale=0.35
    alphaBoolBlend {blendOption=Linear}
    boundBone {boneName=dyn_CordDwon_03}
    componentAppliedLinearAccClamp=<5000 5000 5000>
    componentLinearAccScale=<0.1 0.1 0.1>
    componentLinearVelScale=<0.1 0.1 0.1>
    gravityOverride {z=-980}
    physicsBodyDefinitions [{
      boundBone {boneName=dyn_CordDwon_03}
      boxExtents=<10 30 30>
      constraintSetup {angularLimitsMax=<0 70 10>, angularLimitsMin=<0 -70 -10>, angularTarget=<1 0 0>}
      localJointOffset=<10 0 0>
     }]
    relativeSpaceBone {boneName=Base_Cord}
    simSpaceSettings {maxAngularAcceleration=500, maxAngularVelocity=200, simSpaceAngularAlpha=0.65}
   }
   n8 = "Transform (Modify) Bone - Bone: dyn_CordDwon_01"(ComponentPose=@n7, Alpha=1, Translation="0,-0.5,0.5", Rotation="0,0,0", Scale="1,1,1") {
    translationMode=BMM_Additive
    translationSpace=BCS_BoneSpace
    alphaBoolBlend {blendOption=Linear}
    boneToModify {boneName=dyn_CordDwon_01}
    translation {y=-0.5, z=0.5}
   }
   n9 = "Transform (Modify) Bone - Bone: dyn_CordDwon_02"(ComponentPose=@n8, Alpha=1, Translation="0,0,1", Rotation="0,0,0", Scale="1,1,1") {
    translationMode=BMM_Additive
    translationSpace=BCS_BoneSpace
    alphaBoolBlend {blendOption=Linear}
    boneToModify {boneName=dyn_CordDwon_02}
    translation {z=1}
   }
   n10 = "Transform (Modify) Bone - Bone: dyn_CordDwon_03"(ComponentPose=@n9, Alpha=1, Translation="0,0,1", Rotation="0,0,0", Scale="1,1,1") {
    translationMode=BMM_Additive
    translationSpace=BCS_BoneSpace
    alphaBoolBlend {blendOption=Linear}
    boneToModify {boneName=dyn_CordDwon_03}
    translation {z=1}
   }
   n11 = Component To Local(ComponentPose=@n10)
   n12 = RigidBody(ComponentPose=Local To Component(LocalPose=Use cached pose 'CRCache_InputPose'())) {
    bOverrideWorldGravity=true
    overridePhysicsAsset=&/PineTrimPack_Axe_Physics
    alphaBoolBlend {blendOption=Linear}
    componentAppliedLinearAccClamp=<3000 3000 3000>
    overrideWorldGravity {z=-980}
    simSpaceSettings {
     dampingAlpha=0
     maxAngularAcceleration=50
     maxAngularVelocity=8
     maxLinearAcceleration=2400
     maxLinearVelocity=600
     velocityScaleZ=0.6
     worldAlpha=0.8
     externalLinearDragV=<0.1 0.1 0.1>
    }
   }
   n13 = Component To Local(ComponentPose=@n12)
   n14 = Get SwitchFromRBAN() -> (SwitchFromRBAN:bool)
   n15 = "Transform (Modify) Bone - Bone: Base_Cord"!(
    ComponentPose=Local To Component(LocalPose=Blend Poses by bool(BlendPose_0=@n11, BlendPose_1=@n13, bActiveValue=@n14) {blendType=Linear})
    bAlphaBoolEnabled:bool=Get IsTrailOnPawn() -> (IsTrailOnPawn:bool)
    Translation="0,0,0"
    Rotation="0,110,20"
    Scale="1,1,1"
   ) -> (Pose) {
    alphaInputType=Bool
    rotationMode=BMM_Additive
    alphaBoolBlend {blendOption=Linear}
    boneToModify {boneName=Base_Cord}
    rotation {roll=20, yaw=110}
   }
   n16 = "Transform (Modify) Bone - Bone: Base_Cord"(ComponentPose=@n15, bAlphaBoolEnabled=Get IsNotOnPawn() -> (IsNotOnPawn:bool), Translation="0,0,0", Rotation="0,0,60", Scale="1,1,1") {alphaInputType=Bool, rotationMode=BMM_Additive, alphaBoolBlend {blendOption=Linear}, boneToModify {boneName=Base_Cord}, rotation {roll=60}}
   n17 = Get IsTrailOnPawn()
   n18 = Trail controller - Bone: Dyn_CordUP_03!(ComponentPose=@n16, bAlphaBoolEnabled:bool=@n17, RelaxationSpeedScale:float=4) -> (Pose) {
    alphaInputType=Bool
    bLimitRotation=true
    bShowTrailLocation=true
    chainLength=3
    maxDeltaTime=0.0333
    alphaScaleBiasClamp {bInterpResult=true}
    baseJoint {boneName=b_AR_Root}
    relaxationSpeedScaleInputProcessor {bInterpResult=true, interpSpeedDecreasing=15, interpSpeedIncreasing=15}
    rotationLimits [{limitMax=<10 20 12>, limitMin=<-10 -20 -12>}, {limitMax=<10 20 12>, limitMin=<-10 -20 -12>}, {limitMax=<10 30 15>, limitMin=<-10 -30 -15>}]
    rotationOffsets [<0 0 0>, <0 0 0>, <0 0 0>]
    trailBone {boneName=Dyn_CordUP_03}
    trailRelaxationSpeed {editorCurveData {keys [<at 0 10 RCIM_Linear>,<at 0.4971 7 RCIM_Linear>,<at 1 4 RCIM_Linear>]}}
   }
   n19 = Trail controller - Bone: dyn_CordDwon_03!(ComponentPose=@n18, bAlphaBoolEnabled:bool=@n17, RelaxationSpeedScale:float=4) -> (Pose) {
    alphaInputType=Bool
    bLimitRotation=true
    bShowTrailLocation=true
    chainLength=3
    maxDeltaTime=0.0333
    alphaScaleBiasClamp {bInterpResult=true}
    baseJoint {boneName=b_AR_Root}
    relaxationSpeedScaleInputProcessor {bInterpResult=true, interpSpeedDecreasing=15, interpSpeedIncreasing=15}
    rotationLimits [{limitMax=<10 20 12>, limitMin=<-10 -20 -12>}, {limitMax=<10 20 12>, limitMin=<-10 -20 -12>}, {limitMax=<10 30 15>, limitMin=<-10 -30 -15>}]
    rotationOffsets [<0 0 0>, <0 0 0>, <0 0 0>]
    trailBone {boneName=dyn_CordDwon_03}
    trailRelaxationSpeed {editorCurveData {keys [<at 0 10 RCIM_Linear>,<at 0.4971 7 RCIM_Linear>,<at 1 4 RCIM_Linear>]}}
   }
   n20 = Get IsNotOnPawn()
   n21 = Trail controller - Bone: Dyn_CordUP_03(ComponentPose=@n19, bAlphaBoolEnabled=@n20, RelaxationSpeedScale=3) {
    alphaInputType=Bool
    bLimitRotation=true
    bShowTrailLocation=true
    chainLength=3
    maxDeltaTime=0.0333
    alphaScaleBiasClamp {bInterpResult=true}
    baseJoint {boneName=b_AR_Root}
    relaxationSpeedScaleInputProcessor {bInterpResult=true, interpSpeedDecreasing=15, interpSpeedIncreasing=15}
    rotationLimits [{limitMax=<10 20 12>, limitMin=<-10 -20 -12>}, {limitMax=<10 20 12>, limitMin=<-10 -20 -12>}, {limitMax=<10 30 15>, limitMin=<-10 -30 -15>}]
    rotationOffsets [<0 0 0>, <0 0 0>, <0 0 0>]
    trailBone {boneName=Dyn_CordUP_03}
    trailRelaxationSpeed {editorCurveData {keys [<at 0 10 RCIM_Linear>,<at 0.4971 7 RCIM_Linear>,<at 1 4 RCIM_Linear>]}}
   }
   n22 = Trail controller - Bone: dyn_CordDwon_03(ComponentPose=@n21, bAlphaBoolEnabled=@n20, RelaxationSpeedScale=3) {
    alphaInputType=Bool
    bLimitRotation=true
    bShowTrailLocation=true
    chainLength=3
    maxDeltaTime=0.0333
    alphaScaleBiasClamp {bInterpResult=true}
    baseJoint {boneName=b_AR_Root}
    relaxationSpeedScaleInputProcessor {bInterpResult=true, interpSpeedDecreasing=15, interpSpeedIncreasing=15}
    rotationLimits [{limitMax=<10 20 12>, limitMin=<-10 -20 -12>}, {limitMax=<10 20 12>, limitMin=<-10 -20 -12>}, {limitMax=<10 30 15>, limitMin=<-10 -30 -15>}]
    rotationOffsets [<0 0 0>, <0 0 0>, <0 0 0>]
    trailBone {boneName=dyn_CordDwon_03}
    trailRelaxationSpeed {editorCurveData {keys [<at 0 10 RCIM_Linear>,<at 0.4971 7 RCIM_Linear>,<at 1 4 RCIM_Linear>]}}
   }
   n23 = "Transform (Modify) Bone - Bone: Base_Cord"(ComponentPose=@n22, bAlphaBoolEnabled=Get bIsInFrontEnd(), Translation="0,0,0", Rotation="0,0,-45", Scale="1,1,1") {alphaInputType=Bool, rotationMode=BMM_Additive, alphaBoolBlend {blendOption=Linear}, boneToModify {boneName=Base_Cord}, rotation {roll=-45}}
   Output Pose(Result=Component To Local(ComponentPose=@n23)) {layerGroup=None}
   AnimGraphNode_SubInput_0 = Input Pose()
   Save cached pose 'CRCache_InputPose'(Pose=@AnimGraphNode_SubInput_0) {cacheName=CRCache_InputPose}
  ]}
 EventGraph {nodes [
   Event BlueprintUpdateAnimation()
   n1 = NOT Boolean(A=Get bRigidBodyEnabled())
   > Set SwitchFromRBAN(SwitchFromRBAN:bool=OR Boolean(A=Get bIsInFrontEnd(), B=@n1)) -> (Output_Get:bool)
   n2 = Get bAnimDynamicsEnabled() -> (bAnimDynamicsEnabled:bool)
   > Set SwithToTrailController(SwithToTrailController:bool=AND Boolean(A=@n1, B=NOT Boolean(A=@n2))) -> (Output_Get:bool)
   n3 = AND Boolean(A=Get bIsOnPawn() -> (bIsOnPawn:bool), B=Get SwithToTrailController() -> (SwithToTrailController:bool))
   > Set IsTrailOnPawn(IsTrailOnPawn:bool=@n3) -> (Output_Get:bool)
   > Set IsNotOnPawn(IsNotOnPawn:bool=NOT Boolean(A=@n3)) -> (Output_Get:bool)
  ]}
]
variables [
 Custom Base Mesh {type=SkeletalMeshComponent}
 Is On Pawn {type=boolean}
 IsNotOnPawn {type=boolean}
 IsTrailOnPawn {type=boolean}
 NewVar {type=Pawn, default=null}
 SwitchFromRBAN {type=boolean}
 SwithToTrailController {type=boolean}
]
