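# Physics asset for the PineTrimPack pickaxe, simulating a forked hanging cord on the axe mesh.
# Capsule bodies run from Base_Cord to Dyn_Cord, which branches into three-segment upper and lower chains.
# Each joint uses the same swing limits with a zero twist limit.
# The last upper segment carries a much fatter, offset capsule than the other links.
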PhysicsAsset /BRCosmetics/Weapons/FORT_Melee/Pickaxe_PineTrimPack/Meshes/PineTrimPack_Axe_Physics_FE
bodies {
 Base_Cord [{shapeType=Capsule, length=0.505, radius=0.505, shapeName=Base_Cord_capsule, rotation=<rot 90 0 0>}]
 Dyn_Cord [{shapeType=Capsule, length=1.1462, radius=1, shapeName=Dyn_Cord_capsule, center=<1 0 0>, rotation=<rot 90 0 0>}]
 Dyn_CordUP_01 [{shapeType=Capsule, length=3.8, radius=0.5, shapeName=Dyn_CordUP_01_capsule, center=<1 0 0>, rotation=<rot 90 0 0>}]
 Dyn_CordUP_02 [{shapeType=Capsule, length=3.5218, radius=0.5, shapeName=Dyn_CordUP_02_capsule, center=<1 0.2 0>, rotation=<rot 90 0 0>}]
 Dyn_CordUP_03 [{shapeType=Capsule, length=3.752, radius=3.2849, shapeName=Dyn_CordUP_03_capsule, center=<5.6459 2.3 0>, rotation=<rot 90 0 0>}]
 dyn_CordDwon_01 [{shapeType=Capsule, length=2.5, radius=0.5434, shapeName=dyn_CordDwon_01_capsule, center=<1 0 0>, rotation=<rot 90 0 0>}]
 dyn_CordDwon_02 [{shapeType=Capsule, length=3.4739, radius=0.5, shapeName=dyn_CordDwon_02_capsule, center=<1 0 0>, rotation=<rot 90 0 0>}]
 dyn_CordDwon_03 [{shapeType=Capsule, length=6.6659, radius=0.7, shapeName=dyn_CordDwon_03_capsule, center=<4 0 0>, rotation=<rot 90 0 0>}]
}
constraints [
 {bone1Name=dyn_CordDwon_03, bone2Name=dyn_CordDwon_02, swing1LimitDegrees=60, swing2LimitDegrees=80, twistLimitDegrees=0}
 {bone1Name=Dyn_CordUP_02, bone2Name=Dyn_CordUP_01, swing1LimitDegrees=60, swing2LimitDegrees=80, twistLimitDegrees=0}
 {bone1Name=Dyn_CordUP_03, bone2Name=Dyn_CordUP_02, swing1LimitDegrees=60, swing2LimitDegrees=80, twistLimitDegrees=0}
 {bone1Name=Dyn_CordUP_01, bone2Name=Dyn_Cord, swing1LimitDegrees=60, swing2LimitDegrees=80, twistLimitDegrees=0}
 {bone1Name=dyn_CordDwon_02, bone2Name=dyn_CordDwon_01, swing1LimitDegrees=60, swing2LimitDegrees=80, twistLimitDegrees=0}
 {bone1Name=Dyn_Cord, bone2Name=Base_Cord, swing1LimitDegrees=60, swing2LimitDegrees=80, twistLimitDegrees=0}
 {bone1Name=dyn_CordDwon_01, bone2Name=Dyn_Cord, swing1LimitDegrees=60, swing2LimitDegrees=80, twistLimitDegrees=0}
]
